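# Niagara animation trail effect for the WiryPerk pickaxe, pairing a ribbon trail with small sprite particles, both driven by a system-level socket reader.
# FN_AnimTrail001 uses the shared FN AnimTrail modules to build the ribbon, shifting it from bright yellow-green to teal over its short life.
# BlowingParticles spawns brief particles around a sampled socket, using three partitions to pick yellow, blue or pink, then applies drag, slight inherited velocity and shrinking.
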NiagaraSystem /BRCosmetics/Weapons/FORT_Melee/Pickaxe_WiryPerk/FX/NS_Pickaxe_WiryPerk_AnimTrail
dependencies {
 usedDataInterfaces [
  &NiagaraDataInterfaceColorCurve
  &NiagaraDataInterfaceCurve
  &NiagaraDataInterfaceArrayColor
  &NiagaraDataInterfaceSocketReader
 ]
 usedDynamicInputs []
 usedModules [
  &SystemState
  &self:SystemSpawnScript.SystemScriptSource.SystemScriptGraph.NiagaraNodeAssignment_1.SetVariables_#1
 ]
}
emitters {
 BlowingParticles {
  emitterUpdateScript [
   &EmitterState(Life Cycle Mode=::NewEnumerator1, Loop Behavior=::NewEnumerator1, Loop Duration=0.7, UseLoopDelay=false)
   &SpawnRate(SpawnRate=176.3651, Use Spawn Probability=false)
  ]
  particleSpawnScript [
   &PartitionParticles(Number Of Partitions=3)
   &InitializeParticle(
    Color=&SelectLinearColorFromArray(
     Array Sampling Mode=::NewEnumerator1
     Direct Array Index=$Particles.PartitionParticles.Partition
     Color Selection Array=DataInterface() {propertyValues {ColorData [<rgba 0.9896 0.7693 0 1>, <rgba 0 0.6051 1 1>, <rgba 1 0 0.5747 1>]}}
     Random Seed=?
    )
    Lifetime Max=0.09
    Lifetime Min=0.15
    Lifetime Mode=::NewEnumerator1
    Mass Max=1.5
    Mass Mode=::NewEnumerator3
    Sprite Size Mode=::NewEnumerator3
    Uniform Sprite Size=&RandomRangeFloat(Minimum=1, Maximum=4)
    UsePositionOffset=false
    Write Material Random=true
   )
   &SocketLocation(
    SamplePosOffset=true
    Sampled Position Offset=<-15 0 0>
    Socket Index=2
    Socket Reader=$System.Socket Reader
    Socket Sampling Mode=::NewEnumerator2
    VelocitySpeedLimit=false
   )
   &ShapeLocation(MaskSpawnGroup=false, Override Seed=false, Rotation Mode=::NewEnumerator1, Sphere Radius=15, UseOwnerScale=true)
   &AddVelocity(Module.Use Falloff Distance=false, Override Seed=false, Velocity Mode=::NewEnumerator1)
  ]
  particleUpdateScript [
   &ParticleState()
   &Drag(
    Drag=&FloatFromCurve(
     FloatCurve=DataInterface() {propertyValues {Curve {
        defaultValue=1
        keys [
         {arriveTangent=4.1785, interpMode=RCIM_Cubic, leaveTangent=4.1785, tangentMode=RCTM_User, time=0, value=0}
         {arriveTangent=1.1179, interpMode=RCIM_Cubic, leaveTangent=1.1179, tangentMode=RCTM_User, time=0.2042, value=0.6228}
         {arriveTangent=0.0224, interpMode=RCIM_Cubic, leaveTangent=0.0224, tangentMode=RCTM_User, time=1, value=1}
        ]
       }}}
     Scale Curve=3
    )
    Use Linear Drag=true
    Use Rotational Drag=false
   )
   &InheritVelocity(Inherited Velocity Amount Scale=<0.05 0.05 0>, Inherited Velocity Speed Limit=50, Speed Limit=true, UseSpeedThreshold=false)
   &SolveForcesAndVelocity(Clamp Velocity=false, Limit Acceleration=false)
   &ScaleSpriteSize(Uniform Curve Sprite Scale=DataInterface() {propertyValues {Curve {keys [
        {arriveTangent=1.1111, interpMode=RCIM_Cubic, time=0.1, value=1}
        {arriveTangent=3.3194, interpMode=RCIM_Cubic, leaveTangent=3.3194, tangentMode=RCTM_User, time=0.1767, value=1.1288}
        {arriveTangent=-3.8541, interpMode=RCIM_Cubic, leaveTangent=-3.8541, tangentMode=RCTM_User, time=0.9976, value=-0.0144}
       ]}}})
  ]
  rendererClasses [&NiagaraSpriteRendererProperties]
  renderers [{rendererClass=&NiagaraSpriteRendererProperties}]
 }
 FN_AnimTrail001 {
  emitterUpdateScript [
   &EmitterState(Life Cycle Mode=::NewEnumerator1, Loop Behavior=::NewEnumerator1, Loop Delay=0.05, Loop Duration=1, UseLoopDelay=false)
   &FN_SpawnAnimTrail_v2()
  ]
  particleSpawnScript [
   /Game/Effects/Niagara/Modules/AnimTrail/FN_InitializeAnimTrail/FN_InitializeAnimTrailBoneIndecies(First Bone or Socket Index=1)
   /Game/Effects/Niagara/Modules/AnimTrail/FN_InitializeAnimTrail/FN_InitializeAnimTrail_SocketReader/FN_InitializeAnimTrail_SocketReader(Lifetime=0.2, Socket Reader:NiagaraDataInterfaceSocketReader=$System.Socket Reader)
   &AddVelocity(Velocity=<0 -562.516 0>) {enabled=false}
  ]
  particleUpdateScript [
   &SolveForcesAndVelocity(Clamp Velocity=false, Limit Acceleration=false)
   /Game/Effects/Niagara/Modules/AnimTrail/FN_UpdateAnimTrail/FN_UpdateAnimTrail_SocketReader/FN_UpdateAnimTrail_SocketReader(Socket Reader:NiagaraDataInterfaceSocketReader=$System.Socket Reader)
   &ScaleColor(
    Scale Alpha=&Multiply_Float(
     A=&FloatFromCurve(FloatCurve=DataInterface() {propertyValues {Curve {keys [{arriveTangent=-1.5778, interpMode=RCIM_Cubic, leaveTangent=-1.5778, tangentMode=RCTM_User, time=0.4032, value=1}, <at 1.0005 0.6398>]}}})
     B=1
    )
    Scale RGB=&MakeVectorFromLinearColorRGB(Color=&ColorFromCurve(ColorCurve=DataInterface() {propertyValues {
        AlphaCurve {keys [{interpMode=RCIM_Cubic, leaveTangent=3.0786, time=0.0974, value=1}]}
        BlueCurve {keys [{arriveTangent=1.0842, interpMode=RCIM_Cubic, time=0, value=0}, <at 0.14 0>, <at 0.3775 0.4824 RCIM_Cubic>]}
        GreenCurve {keys [{arriveTangent=0.7094, interpMode=RCIM_Cubic, time=0, value=2}, <at 0.14 1.0338>, <at 0.3775 0.6162 RCIM_Cubic>]}
        RedCurve {keys [{arriveTangent=-700.4397, interpMode=RCIM_Cubic, time=0, value=1.2174}, <at 0.14 0.601>, <at 0.3775 0 RCIM_Cubic>]}
       }}))
    ScaleA=false
    ScaleRGB=true
   )
  ]
  rendererClasses [&NiagaraRibbonRendererProperties]
  renderers [{rendererClass=&NiagaraRibbonRendererProperties}]
 }
}
summary {
 systemName=NS_Pickaxe_WiryPerk_AnimTrail
 emitters [
  {emitterName=FN_AnimTrail001, rendererClasses [&NiagaraRibbonRendererProperties]}
  {emitterName=BlowingParticles, rendererClasses [&NiagaraSpriteRendererProperties]}
 ]
 userVariables [
  User.Loop {type=NiagaraFloat, defaultValue=3.1469, description=""}
  User.RibbonIDOffset {type=NiagaraInt32}
  User.SocketNumber {type=NiagaraInt32}
 ]
}
